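# Physics asset for the M_MED_KnifeGrown male medium player cosmetic body, defining secondary motion for its skirt, knout, losso and handle bones.
# Capsule and box bodies form angle-limited chains from small sphere base anchors on the pelvis and thighs, with the handle constrained directly to the pelvis.
# Pelvis, thigh, right calf and root capsules supply body volumes.
# Free-rotation constraints cross-link neighboring front and side skirt panels, tying them together.
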PhysicsAsset /BRCosmetics/Characters/Player/Male/Medium/Bodies/M_MED_KnifeGrown/Meshes/M_MED_KnifeGrown_Physics
bodies {
 calf_r [{shapeType=Capsule, length=35, radius=6, shapeName=calf_r_capsule, center=<19.1851 -0.1924 -1.4532>, rotation=<rot 90 -0.0001 0>}]
 dyn_handle [{
   shapeType=Capsule
   length=9.2528
   radius=2.1653
   shapeName=dyn_handle_capsule
   center=<10.0082 0.0171 0.2204>
   rotation=<rot 0 95.8775 -87.0547>
  }]
 dyn_knout_01 [{
   shapeType=Capsule
   length=15
   radius=3
   shapeName=dyn_knout_01_capsule
   center=<9.5453 -0.0001 -2.1795>
   rotation=<rot 85.0754 0.2904 0.1377>
  }]
 dyn_knout_02 [{shapeType=Capsule, length=4, radius=4, shapeName=dyn_knout_02_capsule, center=<2.793 0.1971 -3.3579>, rotation=<rot 0.1049 -88.4286 85.9568>}]
 dyn_knout_03 [{
   shapeType=Capsule
   length=15
   radius=4
   shapeName=dyn_knout_03_capsule
   center=<7.7984 -0.1255 -2.2844>
   rotation=<rot 0.0421 -87.5742 -97.7091>
  }]
 dyn_knout_04 [
  {
   shapeType=Capsule
   length=8
   radius=4
   shapeName=dyn_knout_04_capsule
   center=<7.6568 0.9998 -3.0145>
   rotation=<rot -78.881 92.4829 -1.5545>
  }
  {shapeType=Capsule, length=5, radius=2, shapeName=dyn_knout_04_capsule, center=<10.6531 0 -1.144>, rotation=<rot 90 0 0>}
 ]
 dyn_knout_base [{shapeType=Sphere, radius=0.505, shapeName=dyn_knout_base_sphere}]
 dyn_losso_01 [{shapeType=Capsule, length=15, radius=3, shapeName=dyn_losso_01_capsule, center=<8.1853 0 -2.5184>, rotation=<rot 72.664 0.0002 0>}]
 dyn_losso_02 [{shapeType=Capsule, length=15, radius=3, shapeName=dyn_losso_02_capsule, center=<8.1853 0 -2.5184>, rotation=<rot 72.664 0.0002 0>}]
 dyn_skirt_back [{shapeType=Capsule, length=6, radius=2.5, shapeName=dyn_skirt_back_capsule, center=<4.4123 0 -3.1558>, rotation=<rot 87.8296 4.0372 3.8776>}]
 dyn_skirt_back_01 [{shapeType=Capsule, length=10, radius=2.5, shapeName=dyn_skirt_back_01_capsule, center=<6.8759 0 -2.7645>, rotation=<rot 90 0.1583 0>}]
 dyn_skirt_back_basa [{shapeType=Sphere, radius=0.505, shapeName=dyn_skirt_back_basa_sphere}]
 dyn_skirt_back_l [{
   shapeType=Capsule
   length=8
   radius=5
   shapeName=dyn_skirt_back_l_capsule
   center=<4.1934 0.0038 -4.8943>
   rotation=<rot 0.0011 -0.0005 -90.1812>
  }]
 dyn_skirt_back_l_01 [{
   shapeType=Capsule
   length=8
   radius=4.5
   shapeName=dyn_skirt_back_l_01_capsule
   center=<5.3018 0 -5.5131>
   rotation=<rot 0.0011 -0.0005 -90.1812>
  }]
 dyn_skirt_back_l_basa [{shapeType=Sphere, radius=0.505, shapeName=dyn_skirt_back_l_basa_sphere}]
 dyn_skirt_back_r [{
   shapeType=Capsule
   length=8
   radius=4
   shapeName=dyn_skirt_back_r_capsule
   center=<5.3018 0.0039 -4.5619>
   rotation=<rot 0.0013 0.0005 -90.1808>
  }]
 dyn_skirt_back_r_01 [{
   shapeType=Capsule
   length=13
   radius=4.5
   shapeName=dyn_skirt_back_r_01_capsule
   center=<10.2762 -0.2182 -5.0284>
   rotation=<rot -9.0395 0.0044 -112.2091>
  }]
 dyn_skirt_back_r_basa [{shapeType=Sphere, radius=0.505, shapeName=dyn_skirt_back_r_basa_sphere}]
 dyn_skirt_front_l_01 [{
   shapeType=Box
   extentX=12
   extentY=2
   extentZ=14
   shapeName=dyn_skirt_front_l_01_box
   center=<9.2137 -1.3085 -0.3137>
   rotation=<rot 2.5139 2.7255 90>
  }]
 dyn_skirt_front_l_02 [{
   shapeType=Box
   extentX=10
   extentY=4
   extentZ=15
   shapeName=dyn_skirt_front_l_02_box
   center=<10.1094 -1.9854 0.3255>
   rotation=<rot 4.5634 6.2166 89.5211>
  }]
 dyn_skirt_front_l_basa [{shapeType=Sphere, radius=0.505, shapeName=dyn_skirt_front_l_basa_sphere}]
 dyn_skirt_front_r_01 [{
   shapeType=Box
   extentX=10
   extentY=15
   extentZ=2
   shapeName=dyn_skirt_front_r_01_box
   center=<9.3083 2.8044 0.7235>
   rotation=<rot 6.2264 -9.6333 -10.1053>
  }]
 dyn_skirt_front_r_02 [{
   shapeType=Box
   extentX=13
   extentY=4
   extentZ=19
   shapeName=dyn_skirt_front_r_02_box
   center=<6.9632 1.0258 1.1963>
   rotation=<rot 8.3154 -8.4741 -99.2857>
  }]
 dyn_skirt_front_r_basa [{shapeType=Sphere, radius=0.505, shapeName=dyn_skirt_front_r_basa_sphere}]
 dyn_skirt_side_l [{shapeType=Capsule, length=8, radius=6, shapeName=dyn_skirt_side_l_capsule, center=<5.8169 0 -5.5895>, rotation=<rot 0 0 90>}]
 dyn_skirt_side_l_01 [{shapeType=Capsule, length=8, radius=6, shapeName=dyn_skirt_side_l_01_capsule, center=<8.598 0 -4.0803>, rotation=<rot 0 0 90>}]
 dyn_skirt_side_l_basa [{shapeType=Sphere, radius=0.505, shapeName=dyn_skirt_side_l_basa_sphere}]
 dyn_skirt_side_r [{
   shapeType=Box
   extentX=8
   extentY=22
   extentZ=2
   shapeName=dyn_skirt_side_r_box
   center=<6.6374 -2.4986 0.9802>
   rotation=<rot 14.2016 -5.9039 2.8715>
  }]
 dyn_skirt_side_r_01 [{
   shapeType=Box
   extentX=10
   extentY=25
   extentZ=2
   shapeName=dyn_skirt_side_r_01_box
   center=<6.8229 -4.6548 1.1042>
   rotation=<rot 8.7312 -6.6785 -5.5105>
  }]
 dyn_skirt_side_r_basa [{shapeType=Sphere, radius=0.505, shapeName=dyn_skirt_side_r_basa_sphere}]
 pelvis [{shapeType=Capsule, length=10, radius=10.5, center=<-2.5616 -2.2277 0>}]
 root [{shapeType=Capsule, length=50, radius=50, shapeName=root_capsule, center=<0 0.0001 -49.3798>, rotation=<rot -0.0001 -90 90>}]
 thigh_l [{shapeType=Capsule, length=40, radius=8.5, shapeName=thigh_l_capsule, center=<-16.0157 -0.2082 0>, rotation=<rot 90 1.7057 0>}]
 thigh_r [{
   shapeType=Capsule
   length=45
   radius=8
   shapeName=thigh_r_capsule
   center=<15.9839 -0.9024 -0.7172>
   rotation=<rot 88.2162 177.8673 -177.8807>
  }]
}
constraints [
 {
  bone1Name=thigh_l
  bone2Name=pelvis
  swing1LimitDegrees=45
  swing1Motion=Free
  swing2LimitDegrees=45
  swing2Motion=Free
  twistLimitDegrees=45
  twistMotion=Free
 }
 {
  bone1Name=thigh_r
  bone2Name=pelvis
  swing1LimitDegrees=45
  swing1Motion=Free
  swing2LimitDegrees=45
  swing2Motion=Free
  twistLimitDegrees=45
  twistMotion=Free
 }
 {bone1Name=dyn_skirt_side_r, bone2Name=dyn_skirt_side_r_basa, swing1LimitDegrees=20, swing2LimitDegrees=6, twistLimitDegrees=5}
 {bone1Name=dyn_skirt_front_r_02, bone2Name=dyn_skirt_front_r_01, swing1LimitDegrees=5, swing2LimitDegrees=10, twistLimitDegrees=10}
 {bone1Name=dyn_skirt_front_r_01, bone2Name=dyn_skirt_front_r_basa, swing1LimitDegrees=10, swing2LimitDegrees=35, twistLimitDegrees=3}
 {bone1Name=dyn_skirt_back, bone2Name=dyn_skirt_back_basa, swing1LimitDegrees=6, swing2LimitDegrees=30, twistLimitDegrees=3}
 {bone1Name=dyn_skirt_front_l_02, bone2Name=dyn_skirt_front_l_01, swing1LimitDegrees=15, swing2LimitDegrees=10, twistLimitDegrees=10}
 {bone1Name=dyn_skirt_front_l_01, bone2Name=dyn_skirt_front_l_basa, swing1LimitDegrees=20, swing2LimitDegrees=45, twistLimitDegrees=20}
 {bone1Name=dyn_knout_04, bone2Name=dyn_knout_03, swing1LimitDegrees=30, swing2LimitDegrees=40, twistLimitDegrees=10}
 {bone1Name=dyn_knout_02, bone2Name=dyn_knout_01, swing1LimitDegrees=15, swing2LimitDegrees=30, twistLimitDegrees=2}
 {bone1Name=dyn_knout_03, bone2Name=dyn_knout_02, swing1LimitDegrees=20, swing2LimitDegrees=25, twistLimitDegrees=6}
 {bone1Name=dyn_handle, bone2Name=pelvis, swing1LimitDegrees=20, swing2LimitDegrees=20, twistLimitDegrees=2}
 {
  bone1Name=dyn_skirt_side_r_basa
  bone2Name=thigh_r
  swing1LimitDegrees=45
  swing1Motion=Free
  swing2LimitDegrees=45
  swing2Motion=Free
  twistLimitDegrees=45
  twistMotion=Free
 }
 {
  bone1Name=dyn_skirt_front_r_basa
  bone2Name=thigh_r
  swing1LimitDegrees=45
  swing1Motion=Free
  swing2LimitDegrees=45
  swing2Motion=Free
  twistLimitDegrees=45
  twistMotion=Free
 }
 {
  bone1Name=dyn_skirt_back_r_basa
  bone2Name=thigh_r
  swing1LimitDegrees=45
  swing1Motion=Free
  swing2LimitDegrees=45
  swing2Motion=Free
  twistLimitDegrees=45
  twistMotion=Free
 }
 {
  bone1Name=dyn_skirt_side_l_basa
  bone2Name=thigh_l
  swing1LimitDegrees=45
  swing1Motion=Free
  swing2LimitDegrees=45
  swing2Motion=Free
  twistLimitDegrees=45
  twistMotion=Free
 }
 {
  bone1Name=dyn_skirt_front_l_basa
  bone2Name=thigh_l
  swing1LimitDegrees=45
  swing1Motion=Free
  swing2LimitDegrees=45
  swing2Motion=Free
  twistLimitDegrees=45
  twistMotion=Free
 }
 {
  bone1Name=dyn_skirt_back_l_basa
  bone2Name=thigh_l
  swing1LimitDegrees=45
  swing1Motion=Free
  swing2LimitDegrees=45
  swing2Motion=Free
  twistLimitDegrees=45
  twistMotion=Free
 }
 {
  bone1Name=dyn_knout_base
  bone2Name=pelvis
  swing1LimitDegrees=45
  swing1Motion=Free
  swing2LimitDegrees=45
  swing2Motion=Free
  twistLimitDegrees=45
  twistMotion=Free
 }
 {
  bone1Name=dyn_skirt_back_basa
  bone2Name=pelvis
  swing1LimitDegrees=45
  swing1Motion=Free
  swing2LimitDegrees=45
  swing2Motion=Free
  twistLimitDegrees=45
  twistMotion=Free
 }
 {bone1Name=dyn_skirt_back_01, bone2Name=dyn_skirt_back, swing1LimitDegrees=6, swing2LimitDegrees=10, twistLimitDegrees=3}
 {bone1Name=dyn_skirt_back_r_01, bone2Name=dyn_skirt_back_r, swing1LimitDegrees=1, swing2LimitDegrees=20, twistLimitDegrees=8}
 {bone1Name=dyn_skirt_back_r, bone2Name=dyn_skirt_back_r_basa, swing1LimitDegrees=10, swing2LimitDegrees=20, twistLimitDegrees=8}
 {bone1Name=dyn_skirt_back_l_01, bone2Name=dyn_skirt_back_l, swing1LimitDegrees=1, swing2LimitDegrees=20, twistLimitDegrees=6}
 {bone1Name=dyn_skirt_back_l, bone2Name=dyn_skirt_back_l_basa, swing1LimitDegrees=10, swing2LimitDegrees=30, twistLimitDegrees=6}
 {bone1Name=dyn_skirt_side_l_01, bone2Name=dyn_skirt_side_l, swing1LimitDegrees=20, swing2LimitDegrees=20, twistLimitDegrees=30}
 {bone1Name=dyn_skirt_side_l, bone2Name=dyn_skirt_side_l_basa, swing1LimitDegrees=20, swing2LimitDegrees=25, twistLimitDegrees=10}
 {bone1Name=dyn_skirt_side_r_01, bone2Name=dyn_skirt_side_r, swing1LimitDegrees=20, swing2LimitDegrees=10, twistLimitDegrees=7}
 {
  bone1Name=dyn_skirt_side_r
  bone2Name=dyn_skirt_front_r_01
  swing1LimitDegrees=45
  swing1Motion=Free
  swing2LimitDegrees=45
  swing2Motion=Free
  twistLimitDegrees=45
  twistMotion=Free
 }
 {
  bone1Name=dyn_skirt_front_l_02
  bone2Name=dyn_skirt_side_l_01
  swing1LimitDegrees=45
  swing1Motion=Free
  swing2LimitDegrees=45
  swing2Motion=Free
  twistLimitDegrees=45
  twistMotion=Free
 }
 {
  bone1Name=calf_r
  bone2Name=thigh_r
  swing1LimitDegrees=45
  swing1Motion=Free
  swing2LimitDegrees=45
  swing2Motion=Free
  twistLimitDegrees=45
  twistMotion=Free
 }
 {
  bone1Name=dyn_skirt_side_l
  bone2Name=dyn_skirt_front_l_01
  swing1LimitDegrees=45
  swing1Motion=Free
  swing2LimitDegrees=45
  swing2Motion=Free
  twistLimitDegrees=45
  twistMotion=Free
 }
 {
  bone1Name=pelvis
  bone2Name=root
  swing1LimitDegrees=45
  swing1Motion=Free
  swing2LimitDegrees=45
  swing2Motion=Free
  twistLimitDegrees=45
  twistMotion=Free
 }
 {bone1Name=dyn_losso_01, bone2Name=dyn_knout_base, swing1LimitDegrees=5, swing2LimitDegrees=45, twistLimitDegrees=5}
 {bone1Name=dyn_losso_02, bone2Name=dyn_knout_base, swing1LimitDegrees=8, swing2LimitDegrees=45, twistLimitDegrees=5}
 {bone1Name=dyn_knout_01, bone2Name=dyn_knout_base, swing1LimitDegrees=5, swing2LimitDegrees=25, twistLimitDegrees=0}
 {
  bone1Name=dyn_skirt_front_r_02
  bone2Name=dyn_skirt_side_r_01
  swing1LimitDegrees=45
  swing1Motion=Free
  swing2LimitDegrees=45
  swing2Motion=Free
  twistLimitDegrees=45
  twistMotion=Free
 }
]
