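# Post-copy AnimBP for the M_MED_KnifeGrown cosmetic body, adding secondary motion to skirt, knout and losso bones.
# It copies the parent mesh pose, then layers state-driven Control Rig, RigidBody with parameter-driven gravity, AnimDynamics and a front-end skirt nudge.
# When DBNO, output switches to the copied pose with only thigh_LOD offsets.
# A separate DBNO-physics chain is unconnected, and computed LOD flags are unused.
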AnimBlueprint /BRCosmetics/Characters/Player/Male/Medium/Bodies/M_MED_KnifeGrown/Meshes/M_MED_KnifeGrown_AnimBP
target_skeleton=&/M_MED_KnifeGrown_Skeleton
graphs [
 AnimGraph {nodes [
   n1 = "Constraint - dyn_skirt_front_l_basa (R)"(ComponentPose=Local To Component(LocalPose=Use cached pose 'baseCache'()), Alpha=1, ConstraintWeights_0=1) {
    alphaBoolBlend {blendOption=Linear}
    boneToModify {boneName=dyn_skirt_front_l_basa}
    constraintSetup [{transformType=Rotation, targetBone {boneName=thigh_l}}]
    constraintWeights [1]
   }
   n2 = "Constraint - dyn_skirt_front_r_basa (R)"(ComponentPose=@n1, Alpha=1, ConstraintWeights_0=1) {
    alphaBoolBlend {blendOption=Linear}
    boneToModify {boneName=dyn_skirt_front_r_basa}
    constraintSetup [{transformType=Rotation, targetBone {boneName=thigh_r}}]
    constraintWeights [1]
   }
   n3 = "Transform (Modify) Bone - Bone: dyn_skirt_front_l_basa"(ComponentPose=@n2, Alpha=1, Translation="0,0,5", Rotation="0,0,0", Scale="1,1,1") {rotationMode=BMM_Additive, translationMode=BMM_Additive, boneToModify {boneName=dyn_skirt_front_l_basa}, translation {z=5}}
   n4 = "Transform (Modify) Bone - Bone: dyn_skirt_side_l_basa"(ComponentPose=@n3, Alpha=1, Translation="2,0,0", Rotation="10,0,-50", Scale="1,1,1") {
    rotationMode=BMM_Additive
    translationMode=BMM_Additive
    boneToModify {boneName=dyn_skirt_side_l_basa}
    rotation {pitch=10, roll=-50}
    translation {x=2}
   }
   n5 = "Transform (Modify) Bone - Bone: dyn_skirt_front_r_basa"(ComponentPose=@n4, Alpha=1, Translation="0,0,3", Rotation="0,0,0", Scale="1,1,1") {rotationMode=BMM_Additive, translationMode=BMM_Additive, boneToModify {boneName=dyn_skirt_front_r_basa}, translation {z=3}}
   n6 = "Transform (Modify) Bone - Bone: dyn_skirt_side_r_basa"(ComponentPose=@n5, Alpha=1, Translation="-5,0,0", Rotation="10,0,-30", Scale="1,1,1") {
    rotationMode=BMM_Additive
    translationMode=BMM_Additive
    boneToModify {boneName=dyn_skirt_side_r_basa}
    rotation {pitch=10, roll=-30}
    translation {x=-5}
   }
   n7 = "Transform (Modify) Bone - Bone: dyn_knout_base"(ComponentPose=@n6, Alpha=1, Translation="0,0,0", Rotation="-10,-5,-20", Scale="1,1,1") {rotationMode=BMM_Additive, translationMode=BMM_Additive, boneToModify {boneName=dyn_knout_base}, rotation=<rot -10 -5 -20>}
   n8 = "Transform (Modify) Bone - Bone: dyn_knout_02"(ComponentPose=@n7, Alpha=1, Translation="0,0,0", Rotation="20,0,-90", Scale="1,1,1") {rotationMode=BMM_Additive, translationMode=BMM_Additive, boneToModify {boneName=dyn_knout_02}, rotation {pitch=20, roll=-90}}
   n9 = Control Rig!(
    Source=Component To Local(ComponentPose=@n8)
    Alpha
    Sliding:bool=Get bIsSliding()
    Playing emote:bool=Get bIsPlayingEmote()
    crounch:bool=Get bIsCrouching()
    jumping:double=Get Is Jumping() -> (Is Jumping:double)
    SkyDiving:bool=Get bIsSkyDiving()
    Dive mode:bool=Get bIsSkydiveDiveMode()
    DBNO:bool=Get bIsDBNO()
   ) -> (Pose) {controlRigAssetReference {blueprintRigClass=&/M_MED_KnifeGrown_CtrlRig.M_MED_KnifeGrown_CtrlRig_C}}
   n10 = RigidBody!(ComponentPose=Local To Component(LocalPose=@n9), Alpha, OverrideWorldGravity:Vector=Get Gravity Override(), ExternalForce) -> (Pose) {
    bOverrideWorldGravity=true
    overridePhysicsAsset=&/M_MED_KnifeGrown_DBNO_Physics
    simulationSpace=BaseBoneSpace
    alphaBoolBlend {blendOption=Linear}
    baseBoneRef {boneName=pelvis}
    simSpaceSettings {maxAngularAcceleration=500, maxAngularVelocity=3, maxLinearAcceleration=400, maxLinearVelocity=250, velocityScaleZ=0.7, worldAlpha=1}
   }
   n11 = Anim Dynamics - Bone: dyn_losso_01(ComponentPose=@n10, Alpha=0.8, ExternalForce="0,0,0") {
    angularSpringConstant=250
    bAngularSpring=true
    bUseGravityOverride=true
    simulationSpace=RootRelative
    boundBone {boneName=dyn_losso_01}
    componentAppliedLinearAccClamp=<10000 10000 10000>
    componentLinearAccScale=<0.4 0.4 0.4>
    componentLinearVelScale=<0.2 0.2 0.2>
    gravityOverride {z=-1000}
    physicsBodyDefinitions [{
      boundBone {boneName=dyn_losso_01}
      boxExtents=<10 10 10>
      constraintSetup {angularLimitsMax=<0 40 0>, angularLimitsMin=<0 0 0>, angularTarget=<1 0 0>}
      localJointOffset=<1 0 0>
     }]
   }
   n12 = Anim Dynamics - Bone: dyn_losso_02(ComponentPose=@n11, Alpha=1, ExternalForce="0,0,0") {
    angularSpringConstant=800
    bAngularSpring=true
    bUseGravityOverride=true
    simulationSpace=RootRelative
    boundBone {boneName=dyn_losso_02}
    componentAppliedLinearAccClamp=<10000 10000 10000>
    componentLinearAccScale=<0.5 0.5 0.5>
    componentLinearVelScale=<0.2 0.2 0.2>
    gravityOverride {z=-1000}
    physicsBodyDefinitions [{
      boundBone {boneName=dyn_losso_02}
      boxExtents=<10 10 10>
      constraintSetup {angularLimitsMax=<0 40 0>, angularLimitsMin=<0 0 0>, angularTarget=<1 0 0>}
      localJointOffset=<1 0 0>
     }]
   }
   n13 = Anim Dynamics - Bone: dyn_knout_01(ComponentPose=@n12, Alpha=1, ExternalForce="0,0,0") {
    angularSpringConstant=800
    bAngularSpring=true
    bUseGravityOverride=true
    simulationSpace=RootRelative
    boundBone {boneName=dyn_knout_01}
    componentAppliedLinearAccClamp=<10000 10000 10000>
    componentLinearAccScale=<0.5 0.5 0.5>
    componentLinearVelScale=<0.2 0.2 0.2>
    gravityOverride {z=-500}
    physicsBodyDefinitions [{
      boundBone {boneName=dyn_knout_01}
      boxExtents=<10 10 10>
      constraintSetup {angularLimitsMax=<0 6 0>, angularLimitsMin=<0 0 0>, angularTarget=<1 0 0>}
      localJointOffset=<1 0 0>
     }]
   }
   Component To Local(ComponentPose=@n13)
   n14 = Control Rig!(
    Source=Use cached pose 'baseCache'()
    Alpha
    Sliding=Get bIsSliding()
    Playing emote=Get bIsPlayingEmote()
    crounch=Get bIsCrouching()
    jumping=Get Is Jumping()
    SkyDiving=Get bIsSkyDiving()
    Dive mode=Get bIsSkydiveDiveMode()
    DBNO=Get bIsDBNO()
    jumping_b:bool=Get bIsJumping()
   ) -> (Pose) {controlRigAssetReference {blueprintRigClass=&M_MED_KnifeGrown_CtrlRig_C}}
   n15 = Anim Dynamics - Bone: dyn_knout_01(ComponentPose=Local To Component(LocalPose=@n14), Alpha=1, ExternalForce="0,0,0") {
    angularSpringConstant=100
    bAngularSpring=true
    bUseGravityOverride=true
    simulationSpace=RootRelative
    boundBone {boneName=dyn_knout_01}
    componentAppliedLinearAccClamp=<10000 10000 10000>
    componentLinearAccScale=<0.3 0.3 0.3>
    componentLinearVelScale=<0.2 0.2 0.2>
    gravityOverride {z=-1000}
    physicsBodyDefinitions [{
      boundBone {boneName=dyn_knout_01}
      boxExtents=<10 10 10>
      constraintSetup {angularLimitsMax=<0 30 0>, angularLimitsMin=<0 0 0>, angularTarget=<1 0 0>}
      localJointOffset=<1 0 0>
     }]
   }
   n16 = RigidBody(ComponentPose=@n15, Alpha=0.85, OverrideWorldGravity=Get Gravity Override()) {
    bOverrideWorldGravity=true
    overridePhysicsAsset=&/M_MED_KnifeGrown_Physics
    simulationSpace=BaseBoneSpace
    alphaBoolBlend {blendOption=Linear}
    baseBoneRef {boneName=pelvis}
    simSpaceSettings {maxAngularAcceleration=500, maxAngularVelocity=3, maxLinearAcceleration=400, maxLinearVelocity=250, velocityScaleZ=0.7, worldAlpha=1}
   }
   n17 = Anim Dynamics - Bone: dyn_skirt_front_l_02(ComponentPose=@n16, Alpha=1, ExternalForce="0,0,0") {
    angularSpringConstant=100
    bAngularSpring=true
    bUseGravityOverride=true
    simulationSpace=RootRelative
    boundBone {boneName=dyn_skirt_front_l_02}
    componentLinearAccScale=<0.2 0.2 0.2>
    componentLinearVelScale=<0.2 0.2 0.2>
    gravityOverride {z=-350}
    physicsBodyDefinitions [{
      boundBone {boneName=dyn_skirt_front_l_02}
      boxExtents=<12 12 12>
      constraintSetup {angularLimitsMax=<0 90 0>, angularLimitsMin=<0 -5 0>, angularTarget=<1 0 0>}
      localJointOffset=<1 0 0>
     }]
   }
   n18 = Anim Dynamics - Bone: dyn_skirt_front_r_02(ComponentPose=@n17, Alpha=1, ExternalForce="0,0,0") {
    angularSpringConstant=250
    bAngularSpring=true
    bUseGravityOverride=true
    simulationSpace=RootRelative
    boundBone {boneName=dyn_skirt_front_r_02}
    componentLinearAccScale=<0.1 0.1 0.1>
    componentLinearVelScale=<0.2 0.2 0.2>
    gravityOverride {z=-400}
    physicsBodyDefinitions [{
      boundBone {boneName=dyn_skirt_front_r_02}
      boxExtents=<15 15 15>
      constraintSetup {angularLimitsMax=<0 30 0>, angularLimitsMin=<0 0 0>, angularTarget=<1 0 0>}
      localJointOffset=<0.5 0 0>
     }]
   }
   n19 = Anim Dynamics - Bone: dyn_losso_01(ComponentPose=@n18, Alpha=0.8, ExternalForce="0,0,0") {
    angularSpringConstant=250
    bAngularSpring=true
    bUseGravityOverride=true
    simulationSpace=RootRelative
    boundBone {boneName=dyn_losso_01}
    componentAppliedLinearAccClamp=<10000 10000 10000>
    componentLinearAccScale=<0.4 0.4 0.4>
    componentLinearVelScale=<0.2 0.2 0.2>
    gravityOverride {z=-1000}
    physicsBodyDefinitions [{
      boundBone {boneName=dyn_losso_01}
      boxExtents=<10 10 10>
      constraintSetup {angularLimitsMax=<0 40 0>, angularLimitsMin=<0 0 0>, angularTarget=<1 0 0>}
      localJointOffset=<1 0 0>
     }]
   }
   n20 = Anim Dynamics - Bone: dyn_losso_02(ComponentPose=@n19, Alpha=1, ExternalForce="0,0,0") {
    angularSpringConstant=800
    bAngularSpring=true
    bUseGravityOverride=true
    simulationSpace=RootRelative
    boundBone {boneName=dyn_losso_02}
    componentAppliedLinearAccClamp=<10000 10000 10000>
    componentLinearAccScale=<0.5 0.5 0.5>
    componentLinearVelScale=<0.2 0.2 0.2>
    gravityOverride {z=-1000}
    physicsBodyDefinitions [{
      boundBone {boneName=dyn_losso_02}
      boxExtents=<10 10 10>
      constraintSetup {angularLimitsMax=<0 40 0>, angularLimitsMin=<0 0 0>, angularTarget=<1 0 0>}
      localJointOffset=<1 0 0>
     }]
   }
   n21 = "Transform (Modify) Bone - Bone: dyn_skirt_front_l_basa"(ComponentPose=@n20, Alpha=Get Front End() -> (Front End:double), Translation="0,2,0", Rotation="0,0,0", Scale="1,1,1") {rotationMode=BMM_Additive, translationMode=BMM_Additive, boneToModify {boneName=dyn_skirt_front_l_basa}, translation {y=2}}
   n22 = "Transform (Modify) Bone - Bone: thigh_LOD_l"!(
    ComponentPose=Local To Component(LocalPose=Use cached pose 'baseCache'())
    bAlphaBoolEnabled:bool=True
    Translation="0,-5,0"
    Rotation="0,0,0"
    Scale="1,1,1"
   ) -> (Pose) {
    alphaInputType=Bool
    rotationMode=BMM_Additive
    scaleMode=BMM_Additive
    translationMode=BMM_Additive
    alphaBoolBlend {blendOption=Linear}
    boneToModify {boneName=thigh_LOD_l}
    translation {y=-5}
   }
   n23 = "Transform (Modify) Bone - Bone: thigh_LOD_r"!(ComponentPose=@n22, bAlphaBoolEnabled:bool=True, Translation="0,-5,0", Rotation="0,0,0", Scale="1,1,1") -> (Pose) {
    alphaInputType=Bool
    rotationMode=BMM_Additive
    scaleMode=BMM_Additive
    translationMode=BMM_Additive
    alphaBoolBlend {blendOption=Linear}
    boneToModify {boneName=thigh_LOD_r}
    translation {y=-5}
   }
   n24 = "Transform (Modify) Bone - Bone: thigh_LOD_l"(ComponentPose=@n23, bAlphaBoolEnabled=True, Translation="0,-4,5", Rotation="0,0,0", Scale="1,1,1") {
    alphaInputType=Bool
    rotationMode=BMM_Additive
    scaleMode=BMM_Additive
    translationMode=BMM_Additive
    alphaBoolBlend {blendOption=Linear}
    boneToModify {boneName=thigh_LOD_l}
    translation {y=-4, z=5}
   }
   n25 = "Transform (Modify) Bone - Bone: thigh_LOD_r"(ComponentPose=@n24, bAlphaBoolEnabled=True, Translation="0,-5,5", Rotation="0,0,0", Scale="1,1,1") {
    alphaInputType=Bool
    rotationMode=BMM_Additive
    scaleMode=BMM_Additive
    translationMode=BMM_Additive
    alphaBoolBlend {blendOption=Linear}
    boneToModify {boneName=thigh_LOD_r}
    translation {y=-5, z=5}
   }
   Output Pose(Result=Blend Poses by bool(BlendPose_0=Component To Local(ComponentPose=@n25), BlendPose_1=Component To Local(ComponentPose=@n21), bActiveValue=Get bIsDBNO()) {blendType=Linear}) {layerGroup=None}
   Save cached pose 'baseCache'(Pose=Copy Pose From Mesh() {bUseAttachedParent=true}) {cacheName=baseCache}
   Get bIsJumping()
  ]}
 EventGraph {nodes [
   Event BlueprintUpdateAnimation()
   > n1 = Sequence()
   > n2 = "Gravity Override State Function (Parameter Asset) (Nativized)"(
    CharacterPartAnimInstance=Self-Reference()
    GameTimeInSeconds=GetGameTimeInSeconds()
    Parameters=&/M_MED_KnifeGrown_Parameters
   )
   > Set Gravity Override(Gravity Override=@n2)
   Set Is Jumping(execute=@n1.then_1, Is Jumping:double="To Float (Boolean)"(InBool=Get bIsJumping())) -> (Output_Get:double)
   > Set Front End(Front End:double="To Float (Boolean)"(InBool=Get bIsInFrontEnd())) -> (Output_Get:double)
   > Set is LOD2 or Lower(is LOD2 or Lower:bool="integer >= integer"(A=Get CurrentLOD(), B=2)) -> (Output_Get:bool)
   > Set is LOD1 or Lower(is LOD1 or Lower:bool="integer >= integer"(A=Get CurrentLOD(), B=1)) -> (Output_Get:bool)
  ]}
]
variables [
 Front End {type=number}
 Gravity Override {type=Vector}
 Is Jumping {type=number}
 is LOD1 or Lower {type=boolean}
 is LOD2 or Lower {type=boolean}
]
