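# Physics asset for the KnifeGrown male medium face accessory mesh, setting up dynamic hair.
# Capsules on head, upper spine and clavicles form the body volumes around thirteen short hair chains.
# Each chain's base bone attaches to the head, and per-joint swing and twist limits, often asymmetric, shape its bend.
# Extra free-angular constraints tie the back chain's tip to the left and right chain tips, linking those locks.
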PhysicsAsset /BRCosmetics/Characters/Player/Male/Medium/Bodies/M_MED_KnifeGrown/Meshes/Parts/M_MED_KnifeGrown_FaceAcc_Physics
bodies {
 clavicle_l [{shapeType=Capsule, length=10, radius=9, shapeName=clavicle_l_capsule, center=<11.7676 -1.295 0>, rotation=<rot 0 -90 90>}]
 clavicle_r [{shapeType=Capsule, length=10, radius=9, shapeName=clavicle_r_capsule, center=<-11.7818 1.295 -1.3573>, rotation=<rot 0 -90 90>}]
 dyn_hair_back_01 [{shapeType=Capsule, length=4, radius=2, shapeName=dyn_hair_back_01_capsule, center=<0.4357 0 -1.0997>, rotation=<rot 0 0 90>}]
 dyn_hair_back_02 [{shapeType=Capsule, length=4, radius=2.5, shapeName=dyn_hair_back_02_capsule, center=<1.5083 0 -1.3547>, rotation=<rot 0 0 90>}]
 dyn_hair_back_03 [{
   shapeType=Capsule
   length=4
   radius=2.7
   shapeName=dyn_hair_back_03_capsule
   center=<1.4682 -0.0197 -1.5644>
   rotation=<rot 0.0001 0 99.0646>
  }]
 dyn_hair_back_base [{shapeType=Capsule, length=0.505, radius=0.505, shapeName=dyn_hair_back_base_capsule, rotation=<rot 0 -90 0>}]
 dyn_hair_fl_01 [{shapeType=Capsule, length=2, radius=2, shapeName=dyn_hair_fl_01_capsule, rotation=<rot 0 0 90>}]
 dyn_hair_fl_02 [{shapeType=Capsule, length=2, radius=2, shapeName=dyn_hair_fl_02_capsule, rotation=<rot -0.0738 -0.0283 106.8409>}]
 dyn_hair_fl_03 [{shapeType=Capsule, length=2, radius=2, shapeName=dyn_hair_fl_03_capsule, center=<0 -1.1324 -0.7885>, rotation=<rot 0 -0.0001 135.2201>}]
 dyn_hair_fl_04 [{shapeType=Capsule, length=2, radius=2, shapeName=dyn_hair_fl_04_capsule, center=<0 -0.5788 -1.4357>, rotation=<rot 0 -0.0001 125.9839>}]
 dyn_hair_fl_base [{shapeType=Capsule, length=0.505, radius=0.505, shapeName=dyn_hair_fl_base_capsule, rotation=<rot 0 -90 0>}]
 dyn_hair_fr_01 [{shapeType=Capsule, length=2, radius=2, shapeName=dyn_hair_fr_01_capsule, rotation=<rot 0 0 90>}]
 dyn_hair_fr_02 [{shapeType=Capsule, length=2, radius=2, shapeName=dyn_hair_fr_02_capsule, center=<0 0 -0.3207>, rotation=<rot 0 0 90>}]
 dyn_hair_fr_03 [{shapeType=Capsule, length=2, radius=2, shapeName=dyn_hair_fr_03_capsule, center=<0 0.8326 -1.1289>, rotation=<rot 0 0 53.5882>}]
 dyn_hair_fr_04 [{shapeType=Capsule, length=2, radius=2, shapeName=dyn_hair_fr_04_capsule, center=<0 1.4588 -1.9778>, rotation=<rot 0 0 53.5882>}]
 dyn_hair_fr_base [{shapeType=Capsule, length=0.505, radius=0.505, shapeName=dyn_hair_fr_base_capsule, rotation=<rot 0 -90 0>}]
 dyn_hair_front_l_00 [{shapeType=Capsule, length=2, radius=2, shapeName=dyn_hair_front_l_00_capsule, rotation=<rot 0 0 90>}]
 dyn_hair_front_l_01 [{shapeType=Capsule, length=3, radius=3, shapeName=dyn_hair_front_l_01_capsule, center=<0 0.5929 -0.3731>, rotation=<rot 0 0 90>}]
 dyn_hair_front_l_02 [{shapeType=Capsule, length=4, radius=2, shapeName=dyn_hair_front_l_02_capsule, center=<0.2241 2.0179 0>, rotation=<rot -0.0001 -0.0164 0>}]
 dyn_hair_front_l_03 [{
   shapeType=Capsule
   length=6
   radius=3
   shapeName=dyn_hair_front_l_03_capsule
   center=<-0.3033 -1.7076 0.1396>
   rotation=<rot 70.9452 -117.3577 -115.9557>
  }]
 dyn_hair_front_l_base [{shapeType=Capsule, length=0.505, radius=0.505, shapeName=dyn_hair_front_l_base_capsule, rotation=<rot 0 -90 0>}]
 dyn_hair_front_r_01 [{shapeType=Capsule, length=2, radius=2, shapeName=dyn_hair_front_r_01_capsule, rotation=<rot 0 0 90>}]
 dyn_hair_front_r_02 [{shapeType=Capsule, length=4, radius=2, shapeName=dyn_hair_front_r_02_capsule, center=<0 0.315 -0.3306>, rotation=<rot 0 0.0001 10.285>}]
 dyn_hair_front_r_03 [
  {
   shapeType=Capsule
   length=2
   radius=2
   shapeName=dyn_hair_front_r_03_capsule
   center=<1.1035 1.6506 -0.4497>
   rotation=<rot -0.0466 0.1814 13.7399>
  }
  {shapeType=Capsule, length=2, radius=2, shapeName=dyn_hair_front_r_03_capsule, center=<4.091 1.0449 0>}
 ]
 dyn_hair_front_r_base [{shapeType=Capsule, length=0.505, radius=0.505, shapeName=dyn_hair_front_r_base_capsule, rotation=<rot 0 -90 0>}]
 dyn_hair_l_01 [{shapeType=Capsule, length=2, radius=2, shapeName=dyn_hair_l_01_capsule, rotation=<rot 0 0 90>}]
 dyn_hair_l_02 [{shapeType=Capsule, length=2, radius=3, shapeName=dyn_hair_l_02_capsule, center=<0.9831 0 -1.0252>, rotation=<rot 0 0 90>}]
 dyn_hair_l_03 [{shapeType=Capsule, length=4, radius=3, shapeName=dyn_hair_l_03_capsule, center=<1.5455 -0.0242 -2.4303>, rotation=<rot 0.0002 0 89.6952>}]
 dyn_hair_l_base [{shapeType=Capsule, length=0.505, radius=0.505, shapeName=dyn_hair_l_base_capsule, rotation=<rot 0 -90 0>}]
 dyn_hair_local_l2_01 [{shapeType=Capsule, length=2, radius=2, shapeName=dyn_hair_local_l2_01_capsule, rotation=<rot 0 0 90>}]
 dyn_hair_local_l2_02 [{shapeType=Capsule, length=2, radius=2, shapeName=dyn_hair_local_l2_02_capsule, center=<0 0 -0.9047>, rotation=<rot 0.0001 0.0155 90>}]
 dyn_hair_local_l2_03 [{
   shapeType=Capsule
   length=2
   radius=4
   shapeName=dyn_hair_local_l2_03_capsule
   center=<-0.2976 0.353 -1.6599>
   rotation=<rot 0.0027 -0.0012 74.6966>
  }]
 dyn_hair_local_l2_base [{shapeType=Capsule, length=0.505, radius=0.505, shapeName=dyn_hair_local_l2_base_capsule}]
 dyn_hair_local_l_01 [{shapeType=Capsule, length=2, radius=2, shapeName=dyn_hair_local_l_01_capsule, rotation=<rot 0 0 90>}]
 dyn_hair_local_l_02 [{shapeType=Capsule, length=2, radius=5, shapeName=dyn_hair_local_l_02_capsule, rotation=<rot 0 0 90>}]
 dyn_hair_local_l_03 [{shapeType=Capsule, length=2, radius=4.6, shapeName=dyn_hair_local_l_03_capsule, center=<1.6017 0 -1.3104>, rotation=<rot 0 0 90>}]
 dyn_hair_local_l_base [{shapeType=Capsule, length=0.505, radius=0.505, shapeName=dyn_hair_local_l_base_capsule}]
 dyn_hair_local_r2_01 [{shapeType=Capsule, length=2, radius=2, shapeName=dyn_hair_local_r2_01_capsule, rotation=<rot 0 0 90>}]
 dyn_hair_local_r2_02 [{
   shapeType=Capsule
   length=4
   radius=3
   shapeName=dyn_hair_local_r2_02_capsule
   center=<1.7576 0.2438 -0.3742>
   rotation=<rot -0.7748 -84.0531 95.5911>
  }]
 dyn_hair_local_r2_base [{shapeType=Capsule, length=0.505, radius=0.505, shapeName=dyn_hair_local_r2_base_capsule}]
 dyn_hair_local_r_01 [{shapeType=Capsule, length=2, radius=2, shapeName=dyn_hair_local_r_01_capsule, rotation=<rot 0 0 90>}]
 dyn_hair_local_r_02 [{shapeType=Capsule, length=2, radius=4, shapeName=dyn_hair_local_r_02_capsule, center=<3.9276 0.5037 -1.7672>, rotation=<rot 0 0 74.092>}]
 dyn_hair_local_r_base [{shapeType=Capsule, length=0.505, radius=0.505, shapeName=dyn_hair_local_r_base_capsule, rotation=<rot 0 -90 0>}]
 dyn_hair_r_01 [{shapeType=Capsule, length=2, radius=2, shapeName=dyn_hair_r_01_capsule, center=<0 0 -0.76>, rotation=<rot 0 0 90>}]
 dyn_hair_r_02 [{shapeType=Capsule, length=3, radius=2, shapeName=dyn_hair_r_02_capsule, center=<0.1078 0.6639 -1.2129>, rotation=<rot 0 0 81.3992>}]
 dyn_hair_r_03 [{shapeType=Capsule, length=3, radius=2, shapeName=dyn_hair_r_03_capsule, center=<0 0 -1.2878>, rotation=<rot 0 0 90>}]
 dyn_hair_r_base [{shapeType=Capsule, length=0.505, radius=0.505, shapeName=dyn_hair_r_base_capsule, rotation=<rot 0 -90 0>}]
 dyn_hair_side_l_01 [{shapeType=Capsule, length=2, radius=2, shapeName=dyn_hair_side_l_01_capsule, rotation=<rot 0.0321 -0.0004 89.9999>}]
 dyn_hair_side_l_02 [{
   shapeType=Capsule
   length=2
   radius=3
   shapeName=dyn_hair_side_l_02_capsule
   center=<0.3692 0 -1.4934>
   rotation=<rot 0.0316 -0.0005 89.9998>
  }]
 dyn_hair_side_l_03 [{
   shapeType=Capsule
   length=6
   radius=3
   shapeName=dyn_hair_side_l_03_capsule
   center=<-1.8437 0.4122 -2.7401>
   rotation=<rot 0.0315 -0.0005 86.4434>
  }]
 dyn_hair_side_l_base [{shapeType=Capsule, length=0.505, radius=0.505, shapeName=dyn_hair_side_l_base_capsule, rotation=<rot 0 -90 0>}]
 dyn_hair_side_r_01 [{shapeType=Capsule, length=2, radius=2, shapeName=dyn_hair_side_r_01_capsule, rotation=<rot 0 0 90>}]
 dyn_hair_side_r_02 [{shapeType=Capsule, length=2, radius=2, shapeName=dyn_hair_side_r_02_capsule, center=<0.5301 0 -0.6673>, rotation=<rot 0 0 90>}]
 dyn_hair_side_r_03 [{
   shapeType=Capsule
   length=2
   radius=2
   shapeName=dyn_hair_side_r_03_capsule
   center=<-0.8825 -1.3632 -1.3019>
   rotation=<rot 0.0031 0.0069 90.0063>
  }]
 dyn_hair_side_r_base [{shapeType=Capsule, length=0.505, radius=0.505, shapeName=dyn_hair_side_r_base_capsule, rotation=<rot 0 -90 0>}]
 head [{shapeType=Capsule, length=10, radius=8, center=<0 1.1075 0>, rotation=<rot 90 179.9996 180>}]
 spine_05 [{shapeType=Capsule, length=16, radius=10.5, shapeName=spine_05_capsule, center=<8.4021 -1.8376 0>, rotation=<rot 0 90 0>}]
}
constraints [
 {
  bone1Name=dyn_hair_front_l_base
  bone2Name=head
  swing1LimitDegrees=45
  swing1Motion=Free
  swing2LimitDegrees=45
  swing2Motion=Free
  twistLimitDegrees=45
  twistMotion=Free
 }
 {bone1Name=dyn_hair_front_l_01, bone2Name=dyn_hair_front_l_00, swing1LimitDegrees=30, swing2LimitDegrees=6, twistLimitDegrees=5}
 {bone1Name=dyn_hair_front_l_02, bone2Name=dyn_hair_front_l_01, swing1LimitDegrees=40, swing2LimitDegrees=10, twistLimitDegrees=6}
 {
  bone1Name=dyn_hair_front_r_base
  bone2Name=head
  swing1LimitDegrees=45
  swing1Motion=Free
  swing2LimitDegrees=45
  swing2Motion=Free
  twistLimitDegrees=45
  twistMotion=Free
 }
 {bone1Name=dyn_hair_front_r_01, bone2Name=dyn_hair_front_r_base, swing1LimitDegrees=20, swing2LimitDegrees=10, twistLimitDegrees=2}
 {bone1Name=dyn_hair_front_r_02, bone2Name=dyn_hair_front_r_01, swing1LimitDegrees=45, swing2LimitDegrees=10, twistLimitDegrees=2}
 {bone1Name=dyn_hair_r_base, bone2Name=head, swing1LimitDegrees=45, swing2LimitDegrees=45, twistLimitDegrees=45}
 {bone1Name=dyn_hair_r_01, bone2Name=dyn_hair_r_base, swing1LimitDegrees=5, swing2LimitDegrees=65, twistLimitDegrees=2}
 {bone1Name=dyn_hair_r_02, bone2Name=dyn_hair_r_01, swing1LimitDegrees=5, swing2LimitDegrees=65, twistLimitDegrees=30}
 {
  bone1Name=head
  bone2Name=spine_05
  swing1LimitDegrees=45
  swing1Motion=Free
  swing2LimitDegrees=45
  swing2Motion=Free
  twistLimitDegrees=45
  twistMotion=Free
 }
 {
  bone1Name=dyn_hair_back_base
  bone2Name=head
  swing1LimitDegrees=45
  swing1Motion=Free
  swing2LimitDegrees=45
  swing2Motion=Free
  twistLimitDegrees=45
  twistMotion=Free
 }
 {bone1Name=dyn_hair_back_01, bone2Name=dyn_hair_back_base, swing1LimitDegrees=5, swing2LimitDegrees=90, twistLimitDegrees=5}
 {bone1Name=dyn_hair_back_02, bone2Name=dyn_hair_back_01, swing1LimitDegrees=5, swing2LimitDegrees=10, twistLimitDegrees=5}
 {bone1Name=dyn_hair_back_03, bone2Name=dyn_hair_back_02, swing1LimitDegrees=0, swing2LimitDegrees=2, twistLimitDegrees=0}
 {bone1Name=dyn_hair_fl_04, bone2Name=dyn_hair_fl_03, swing1LimitDegrees=15, swing2LimitDegrees=15, twistLimitDegrees=3}
 {
  bone1Name=dyn_hair_fl_base
  bone2Name=head
  swing1LimitDegrees=45
  swing1Motion=Free
  swing2LimitDegrees=45
  swing2Motion=Free
  twistLimitDegrees=45
  twistMotion=Free
 }
 {bone1Name=dyn_hair_fl_01, bone2Name=dyn_hair_fl_base, swing1LimitDegrees=25, swing2LimitDegrees=25, twistLimitDegrees=5}
 {bone1Name=dyn_hair_fl_02, bone2Name=dyn_hair_fl_01, swing1LimitDegrees=5, swing2LimitDegrees=5, twistLimitDegrees=3}
 {bone1Name=dyn_hair_fl_03, bone2Name=dyn_hair_fl_02, swing1LimitDegrees=15, swing2LimitDegrees=15, twistLimitDegrees=3}
 {
  bone1Name=dyn_hair_fr_base
  bone2Name=head
  swing1LimitDegrees=45
  swing1Motion=Free
  swing2LimitDegrees=45
  swing2Motion=Free
  twistLimitDegrees=45
  twistMotion=Free
 }
 {bone1Name=dyn_hair_fr_01, bone2Name=dyn_hair_fr_base, swing1LimitDegrees=25, swing2LimitDegrees=25, twistLimitDegrees=5}
 {bone1Name=dyn_hair_fr_02, bone2Name=dyn_hair_fr_01, swing1LimitDegrees=15, swing2LimitDegrees=15, twistLimitDegrees=5}
 {bone1Name=dyn_hair_fr_03, bone2Name=dyn_hair_fr_02, swing1LimitDegrees=15, swing2LimitDegrees=15, twistLimitDegrees=5}
 {bone1Name=dyn_hair_fr_04, bone2Name=dyn_hair_fr_03, swing1LimitDegrees=15, swing2LimitDegrees=15, twistLimitDegrees=5}
 {bone1Name=dyn_hair_front_l_03, bone2Name=dyn_hair_front_l_02, swing1LimitDegrees=25, swing2LimitDegrees=25, twistLimitDegrees=5}
 {bone1Name=dyn_hair_front_r_03, bone2Name=dyn_hair_front_r_02, swing1LimitDegrees=45, swing2LimitDegrees=10, twistLimitDegrees=1}
 {
  bone1Name=dyn_hair_local_r_base
  bone2Name=head
  swing1LimitDegrees=45
  swing1Motion=Free
  swing2LimitDegrees=45
  swing2Motion=Free
  twistLimitDegrees=45
  twistMotion=Free
 }
 {
  bone1Name=dyn_hair_side_l_base
  bone2Name=head
  swing1LimitDegrees=45
  swing1Motion=Free
  swing2LimitDegrees=45
  swing2Motion=Free
  twistLimitDegrees=45
  twistMotion=Free
 }
 {bone1Name=dyn_hair_side_l_01, bone2Name=dyn_hair_side_l_base, swing1LimitDegrees=2, swing2LimitDegrees=60, twistLimitDegrees=2}
 {bone1Name=dyn_hair_side_l_02, bone2Name=dyn_hair_side_l_01, swing1LimitDegrees=2, swing2LimitDegrees=90, twistLimitDegrees=2}
 {bone1Name=dyn_hair_side_l_03, bone2Name=dyn_hair_side_l_02, swing1LimitDegrees=5, swing2LimitDegrees=60, twistLimitDegrees=2}
 {bone1Name=dyn_hair_r_03, bone2Name=dyn_hair_r_02, swing1LimitDegrees=0, swing2LimitDegrees=65, twistLimitDegrees=0}
 {bone1Name=dyn_hair_local_r_02, bone2Name=dyn_hair_local_r_01, swing1LimitDegrees=20, swing2LimitDegrees=20, twistLimitDegrees=2}
 {bone1Name=dyn_hair_local_r_01, bone2Name=dyn_hair_local_r_base, swing1LimitDegrees=20, swing2LimitDegrees=20, twistLimitDegrees=2}
 {bone1Name=dyn_hair_side_r_03, bone2Name=dyn_hair_side_r_02, swing1LimitDegrees=15, swing2LimitDegrees=70, twistLimitDegrees=3}
 {
  bone1Name=dyn_hair_side_r_base
  bone2Name=head
  swing1LimitDegrees=45
  swing1Motion=Free
  swing2LimitDegrees=45
  swing2Motion=Free
  twistLimitDegrees=45
  twistMotion=Free
 }
 {bone1Name=dyn_hair_side_r_01, bone2Name=dyn_hair_side_r_base, swing1LimitDegrees=5, swing2LimitDegrees=15, twistLimitDegrees=3}
 {bone1Name=dyn_hair_side_r_02, bone2Name=dyn_hair_side_r_01, swing1LimitDegrees=15, swing2LimitDegrees=40, twistLimitDegrees=15}
 {bone1Name=dyn_hair_l_03, bone2Name=dyn_hair_l_02, swing1LimitDegrees=10, swing2LimitDegrees=30, twistLimitDegrees=1}
 {
  bone1Name=dyn_hair_l_base
  bone2Name=head
  swing1LimitDegrees=45
  swing1Motion=Free
  swing2LimitDegrees=45
  swing2Motion=Free
  twistLimitDegrees=45
  twistMotion=Free
 }
 {bone1Name=dyn_hair_l_01, bone2Name=dyn_hair_l_base, swing1LimitDegrees=2, swing2LimitDegrees=60, twistLimitDegrees=1}
 {bone1Name=dyn_hair_l_02, bone2Name=dyn_hair_l_01, swing1LimitDegrees=2, swing2LimitDegrees=50, twistLimitDegrees=1}
 {bone1Name=dyn_hair_front_l_00, bone2Name=dyn_hair_front_l_base, swing1LimitDegrees=25, swing2LimitDegrees=2, twistLimitDegrees=2}
 {
  bone1Name=clavicle_r
  bone2Name=spine_05
  swing1LimitDegrees=45
  swing1Motion=Free
  swing2LimitDegrees=45
  swing2Motion=Free
  twistLimitDegrees=45
  twistMotion=Free
 }
 {
  bone1Name=clavicle_l
  bone2Name=spine_05
  swing1LimitDegrees=45
  swing1Motion=Free
  swing2LimitDegrees=45
  swing2Motion=Free
  twistLimitDegrees=45
  twistMotion=Free
 }
 {bone1Name=dyn_hair_local_r2_02, bone2Name=dyn_hair_local_r2_01, swing1LimitDegrees=16, swing2LimitDegrees=25, twistLimitDegrees=2}
 {
  bone1Name=dyn_hair_local_l2_base
  bone2Name=head
  swing1LimitDegrees=45
  swing1Motion=Free
  swing2LimitDegrees=45
  swing2Motion=Free
  twistLimitDegrees=45
  twistMotion=Free
 }
 {bone1Name=dyn_hair_local_l2_01, bone2Name=dyn_hair_local_l2_base, swing1LimitDegrees=15, swing2LimitDegrees=6, twistLimitDegrees=2}
 {bone1Name=dyn_hair_local_l2_02, bone2Name=dyn_hair_local_l2_01, swing1LimitDegrees=15, swing2LimitDegrees=25, twistLimitDegrees=2}
 {bone1Name=dyn_hair_local_l2_03, bone2Name=dyn_hair_local_l2_02, swing1LimitDegrees=15, swing2LimitDegrees=5, twistLimitDegrees=10}
 {
  bone1Name=dyn_hair_local_l_base
  bone2Name=head
  swing1LimitDegrees=45
  swing1Motion=Free
  swing2LimitDegrees=45
  swing2Motion=Free
  twistLimitDegrees=45
  twistMotion=Free
 }
 {bone1Name=dyn_hair_local_l_01, bone2Name=dyn_hair_local_l_base, swing1LimitDegrees=25, swing2LimitDegrees=25, twistLimitDegrees=0}
 {bone1Name=dyn_hair_local_l_02, bone2Name=dyn_hair_local_l_01, swing1LimitDegrees=25, swing2LimitDegrees=25, twistLimitDegrees=0}
 {bone1Name=dyn_hair_local_l_03, bone2Name=dyn_hair_local_l_02, swing1LimitDegrees=25, swing2LimitDegrees=25, twistLimitDegrees=0}
 {
  bone1Name=dyn_hair_local_r2_base
  bone2Name=head
  swing1LimitDegrees=45
  swing1Motion=Free
  swing2LimitDegrees=45
  swing2Motion=Free
  twistLimitDegrees=45
  twistMotion=Free
 }
 {bone1Name=dyn_hair_local_r2_01, bone2Name=dyn_hair_local_r2_base, swing1LimitDegrees=16, swing2LimitDegrees=25, twistLimitDegrees=2}
 {
  bone1Name=dyn_hair_back_03
  bone2Name=dyn_hair_l_03
  swing1LimitDegrees=45
  swing1Motion=Free
  swing2LimitDegrees=45
  swing2Motion=Free
  twistLimitDegrees=45
  twistMotion=Free
 }
 {
  bone1Name=dyn_hair_back_03
  bone2Name=dyn_hair_r_03
  swing1LimitDegrees=45
  swing1Motion=Free
  swing2LimitDegrees=45
  swing2Motion=Free
  twistLimitDegrees=45
  twistMotion=Free
 }
]
