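# Post-process style AnimBlueprint for the KnifeGrown male character's face accessory part, adding dynamic motion to hair and other accessory bones.
# Copies the parent mesh pose, then runs a Control Rig fed by crouch, skydive and weapon-aim states.
# The cached result feeds three head-space RigidBody variants with offset front-hair bone and different physics assets, combined by bool blends.
# Update logic sets simulation gravity from skydive state and aim flags from weapon type.
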AnimBlueprint /BRCosmetics/Characters/Player/Male/Medium/Bodies/M_MED_KnifeGrown/Meshes/Parts/M_MED_KnifeGrown_FaceAcc_AnimBP
target_skeleton=&/M_MED_KnifeGrown_FaceAcc_Skeleton
graphs [
 AnimGraph {nodes [
   n1 = "Transform (Modify) Bone - Bone: dyn_hair_front_l_base"(ComponentPose=Local To Component(LocalPose=Use cached pose 'PreDyn'()), Alpha=1, Translation="0.6,-0.4,0", Rotation="0,0,0", Scale="1,1,1") {rotationMode=BMM_Additive, translationMode=BMM_Additive, boneToModify {boneName=dyn_hair_front_l_base}, translation=<0.6 -0.4>}
   n2 = Get buckle_Gravity Override() -> (buckle_Gravity Override:Vector)
   n3 = RigidBody!(ComponentPose=@n1, Alpha, OverrideWorldGravity:Vector=@n2, ExternalForce) -> (Pose) {
    bClampLinearTranslationLimitToRefPose=true
    bOverrideWorldGravity=true
    overridePhysicsAsset=&/M_MED_KnifeGrown_FaceAcc_Physics1
    simulationSpace=BaseBoneSpace
    baseBoneRef {boneName=head}
    componentAppliedLinearAccClamp=<3000 3000 3000>
    simSpaceSettings {
     maxAngularAcceleration=32
     maxAngularVelocity=8
     maxLinearAcceleration=1200
     maxLinearVelocity=800
     velocityScaleZ=0.5
     worldAlpha=1
     externalLinearDragV=<0.25 0.25 0.25>
    }
   }
   n4 = Component To Local(ComponentPose=@n3)
   n5 = "Transform (Modify) Bone - Bone: dyn_hair_front_l_base"(ComponentPose=Local To Component(LocalPose=Use cached pose 'PreDyn'()), Alpha=1, Translation="0.6,-0.4,0", Rotation="0,0,0", Scale="1,1,1") {rotationMode=BMM_Additive, translationMode=BMM_Additive, boneToModify {boneName=dyn_hair_front_l_base}, translation=<0.6 -0.4>}
   n6 = RigidBody(ComponentPose=@n5, OverrideWorldGravity=@n2) {
    bClampLinearTranslationLimitToRefPose=true
    bOverrideWorldGravity=true
    overridePhysicsAsset=&/M_MED_KnifeGrown_FaceAcc_Physics
    simulationSpace=BaseBoneSpace
    baseBoneRef {boneName=head}
    componentAppliedLinearAccClamp=<3000 3000 3000>
    simSpaceSettings {
     maxAngularAcceleration=32
     maxAngularVelocity=8
     maxLinearAcceleration=1050
     maxLinearVelocity=800
     velocityScaleZ=0.5
     worldAlpha=1
     externalLinearDragV=<0.25 0.25 0.25>
    }
   }
   n7 = Component To Local(ComponentPose=@n6)
   n8 = "Transform (Modify) Bone - Bone: dyn_hair_front_l_base"(
    ComponentPose=Local To Component(LocalPose=Use cached pose 'PreDyn'())
    Alpha=1
    Translation="-0.2,0,0"
    Rotation="0.000,8.233,-0.000"
    Scale="1,1,1"
   ) {
    rotationMode=BMM_Additive
    translationMode=BMM_Additive
    boneToModify {boneName=dyn_hair_front_l_base}
    rotation {yaw=8.233}
    translation {x=-0.2}
   }
   n9 = RigidBody(ComponentPose=@n8, OverrideWorldGravity=Get buckle_Gravity Override()) {
    bClampLinearTranslationLimitToRefPose=true
    bOverrideWorldGravity=true
    overridePhysicsAsset=&/M_MED_KnifeGrown_FaceAcc_FE_Physics
    simulationSpace=BaseBoneSpace
    baseBoneRef {boneName=head}
    componentAppliedLinearAccClamp=<3000 3000 3000>
    simSpaceSettings {
     maxAngularAcceleration=32
     maxAngularVelocity=8
     maxLinearAcceleration=1050
     maxLinearVelocity=800
     velocityScaleZ=0.5
     worldAlpha=1
     externalLinearDragV=<0.25 0.25 0.25>
    }
   }
   Output Pose(Result=Blend Poses by bool(BlendPose_0=Component To Local(ComponentPose=@n9), BlendPose_1=Blend Poses by bool(BlendPose_0=@n4, BlendPose_1=@n7))) {layerGroup=None}
   n10 = Control Rig!(
    Source=Copy Pose From Mesh() {bUseAttachedParent=true}
    Alpha
    crouching:bool=Get bIsCrouching()
    SkyDiving:bool=Get bIsSkyDiving()
    PistolTaargeting:bool=Get Pistol Equipped() -> (Pistol Equipped:bool)
    LongWTargeting:bool=Get LongW Equipped() -> (LongW Equipped:bool)
   ) -> (Pose) {controlRigAssetReference {blueprintRigClass=&/M_MED_KnifeGrown_FaceAcc_CtrlRig.M_MED_KnifeGrown_FaceAcc_CtrlRig_C}}
   Save cached pose 'PreDyn'(Pose=@n10) {cacheName=PreDyn}
   Get Gravity Override()
  ]}
 EventGraph {nodes [
   n1 = "Gravity Override State Function (Parameter Asset) (Nativized)"(
    CharacterPartAnimInstance=Self-Reference()
    GameTimeInSeconds=GetGameTimeInSeconds()
    Parameters=/BRCosmetics/Characters/Player/Female/Medium/Bodies/F_MED_WriteWalrus/Meshes/M_MED_WriteWalrus_Parameters
   )
   > Set Gravity Override(Gravity Override=@n1)
   Event BlueprintUpdateAnimation()
   > n2 = Sequence()
   n3 = Get bIsSkydiveDiveMode()
   > n4 = Branch(Condition=OR Boolean(A=@n3, B=Get bIsSkydiveFloating()))
   > n5 = Branch(Condition=@n3)
   > Set buckle_Gravity Override(buckle_Gravity Override:Vector="0, 0, 0") -> (Output_Get:Vector)
   Set buckle_Gravity Override(execute=@n5.else, buckle_Gravity Override="0,0,-1500")
   Set buckle_Gravity Override(execute=@n4.else, buckle_Gravity Override="0,0,-1100")
   n6 = Cast To FortPlayerAnimInstance(execute=@n2.then_1, Object=GetAnimInstance(self=GetOwnerSkeletalMeshComponent()))
   > Set LongW Equipped(LongW Equipped:bool=false) -> (Output_Get:bool)
   > n7 = Switch on EFortWeaponCoreAnimation!(execute, Selection:EFortWeaponCoreAnimation Enum=Get WeaponCoreAnim(self=@n6)) -> (
    Melee:Exec
    Pistol:Exec
    Shotgun:Exec
    PaperBlueprint:Exec
    Rifle:Exec
    MeleeOneHand:Exec
    MachinePistol:Exec
    RocketLauncher:Exec
    GrenadeLauncher:Exec
    GoingCommando:Exec
    AssaultRifle:Exec
    TacticalShotgun:Exec
    SniperRifle:Exec
    TrapPlacement:Exec
    ShoulderLauncher:Exec
    AbilityDecoTool:Exec
    Crossbow:Exec
    C4:Exec
    RemoteControl:Exec
    DualWield:Exec
    AR_BullPup:Exec
    AR_ForwardGrip:Exec
    MedPackPaddles:Exec
    SMG_P90:Exec
    AR_DrumGun:Exec
    Consumable_Small:Exec
    Consumable_Large:Exec
    Balloon:Exec
    MountedTurret:Exec
    CreativeTool:Exec
    ExplosiveBow:Exec
    AshtonIndigo:Exec
    AshtonChicago:Exec
    MeleeDualWield:Exec
    Unarmed:Exec
   )
   n8 = Branch(
    execute=[@n7.AssaultRifle, @n7.GoingCommando, @n7.GrenadeLauncher, @n7.Rifle, @n7.SMG_P90, @n7.Shotgun, @n7.SniperRifle, @n7.TacticalShotgun]
    Condition=Get bIsTargeting()
   )
   > Set LongW Equipped(LongW Equipped=true)
   Set LongW Equipped(execute=@n8.else, LongW Equipped=false)
   n9 = Branch(execute=@n7.Pistol, Condition=Get bIsTargeting())
   > Set Pistol Equipped(Pistol Equipped:bool=true) -> (Output_Get:bool)
   Set Pistol Equipped(execute=@n9.else, Pistol Equipped=false)
  ]}
]
variables [
 buckle_Gravity Override {type=Vector}
 Gravity Override {type=Vector}
 LongW Equipped {type=boolean}
 Pistol Equipped {type=boolean}
]
